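# Physics asset for the F_MED_Street_Fashion_Holiday backpack cosmetic, defining collision and joint limits for its dynamic axe bones.
# The dyn_axe body pairs a long thin capsule with an offset box to approximate the axe, while dyn_axe_en is a tiny zero-length capsule.
# A single tightly limited swing and twist constraint links those two bodies.
# A wide capsule on the pack bone covers the bag and has no constraint to the axe.
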PhysicsAsset /BRCosmetics/Accessories/FORT_Backpacks/Backpack_F_MED_Street_Fashion_Holiday/Meshes/F_MED_Street_Fashion_Holiday_Pack_Physics
bodies {
 dyn_axe [
  {shapeType=Capsule, length=30.6409, radius=1.3193, center=<0.0404 0.427 -0.3193>, rotation=<rot -0.7181 -89.9945 0.1217>}
  {shapeType=Box, extentX=1.832, extentY=15.2345, extentZ=7.0323, center=<0 0.1103 17.5459>}
 ]
 dyn_axe_en [{shapeType=Capsule, length=0, radius=0.505, rotation=<rot 0 -90 0>}]
 pack [{shapeType=Capsule, length=18.6773, radius=9.6349, center=<-0.2864 -6.0449 -5.2223>, rotation=<rot -0.7395 -90.0352 -0.2775>}]
}
constraints [{bone1Name=dyn_axe, bone2Name=dyn_axe_en, swing1LimitDegrees=2, swing2LimitDegrees=0.5, twistLimitDegrees=10}]
